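# Sound cue for the final demolition impact on a metal punching bag, routed through the Fort impacts sound class with medium-object destroy attenuation.
# A mixer layers three randomized wave sets: punching bag final impacts, small metal death sounds at reduced volume, and default explosive impact layers.
# The bag layer has explicit pitch variation, while the metal layer passes through a modulator at default settings.
# Concurrency is capped through a shared object-impacts setting, with stop-oldest overrides also present.
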SoundCue /Game/Sounds/Fort_Impact_Sounds/Demolition/PunchBag/Demo_Metal_PunchBag_ImpactFinal_Cue
root=Mixer_3
nodes [
 {id=Mixer_3, children [Modulator_5,Modulator_6,Random_19], properties {inputVolume [1,0.5,1]}}
 {id=Modulator_5, children [Random_9], properties {pitchMax=1.07, pitchMin=0.92}}
 {id=Modulator_6, children [Random_10]}
 {id=Random_10, children [WavePlayer_27, WavePlayer_28, WavePlayer_29], properties {weights [1,1,1]}}
 {id=Random_19, children [WavePlayer_59, WavePlayer_60], properties {weights [1,1]}}
 {id=Random_9, children [WavePlayer_37, WavePlayer_38], properties {weights [1,1]}}
 {id=WavePlayer_27, properties {soundWaveAssetPtr=/Game/Sounds/Fort_Impact_Sounds/Demolition/Metal/Metal_Small_Death03}}
 {id=WavePlayer_28, properties {soundWaveAssetPtr=/Game/Sounds/Fort_Impact_Sounds/Demolition/Metal/Metal_Small_Death02}}
 {id=WavePlayer_29, properties {soundWaveAssetPtr=/Game/Sounds/Fort_Impact_Sounds/Demolition/Metal/Metal_Small_Death01}}
 {id=WavePlayer_37, properties {soundWaveAssetPtr=/Game/Sounds/Fort_Impact_Sounds/Demolition/Gym/Gym_PunchingBag_ImpactFinal05}}
 {id=WavePlayer_38, properties {soundWaveAssetPtr=/Game/Sounds/Fort_Impact_Sounds/Demolition/Gym/Gym_PunchingBag_ImpactFinal06}}
 {id=WavePlayer_59, properties {soundWaveAssetPtr=/Game/Sounds/Fort_Impact_Sounds/Demolition/Defaults/Explo_Impact_Layer01}}
 {id=WavePlayer_60, properties {soundWaveAssetPtr=/Game/Sounds/Fort_Impact_Sounds/Demolition/Defaults/Explo_Impact_Layer02}}
]
properties {
 attenuationSettings=/Game/Sounds/Attenuation/Fort_Demolition_Destroy_Medium_Sized_Object_Att
 maxDistance=4000
 soundClassObject=&Fort_Impacts_PSM
 volumeMultiplier=0.544
 concurrencyOverrides {maxCount=3, resolutionRule=StopOldest}
 concurrencySet [&CONC_Object_Impacts_FourMax]
}
