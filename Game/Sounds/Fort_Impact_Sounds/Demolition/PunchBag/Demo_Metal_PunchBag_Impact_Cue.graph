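# Sound cue for impacts on a punching bag, filed among the demolition impact sounds.
# A random node picks one of four equally weighted gym punching bag waves, then passes it through a modulator.
# It plays in the impacts sound class at reduced volume with small melee weapon attenuation.
# An object-impact concurrency set is assigned, alongside override values that stop the oldest voice.
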SoundCue /Game/Sounds/Fort_Impact_Sounds/Demolition/PunchBag/Demo_Metal_PunchBag_Impact_Cue
root=Modulator_7
nodes [
 {id=Modulator_7, children [Random_12]}
 {id=Random_12, children [WavePlayer_39, WavePlayer_40, WavePlayer_41, WavePlayer_84], properties {weights [1,1,1,1]}}
 {id=WavePlayer_39, properties {soundWaveAssetPtr=/Game/Sounds/Fort_Impact_Sounds/Demolition/Gym/Gym_PunchingBag_Impact05}}
 {id=WavePlayer_40, properties {soundWaveAssetPtr=/Game/Sounds/Fort_Impact_Sounds/Demolition/Gym/Gym_PunchingBag_Impact03}}
 {id=WavePlayer_41, properties {soundWaveAssetPtr=/Game/Sounds/Fort_Impact_Sounds/Demolition/Gym/Gym_PunchingBag_Impact04}}
 {id=WavePlayer_84, properties {soundWaveAssetPtr=/Game/Sounds/Fort_Impact_Sounds/Demolition/Gym/Gym_PunchingBag_Impact06}}
]
properties {
 attenuationSettings=&Fort_Weapon_Melee_Sml_Att
 maxDistance=2800
 soundClassObject=&Fort_Impacts
 volumeMultiplier=0.34
 concurrencyOverrides {maxCount=3, resolutionRule=StopOldest}
 concurrencySet [&CONC_Object_Impacts_FourMax]
}
